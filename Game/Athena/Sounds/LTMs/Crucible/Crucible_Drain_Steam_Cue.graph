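# Sound cue for a looping drain steam ambience in the Crucible LTM, routed to the Athena ambient sound class.
# A single wave player loops one steam sample at reduced volume.
# The cue overrides attenuation itself, using natural-sound falloff from a small inner radius, with stereo spread for spatial width.
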SoundCue /Game/Athena/Sounds/LTMs/Crucible/Crucible_Drain_Steam_Cue
root=WavePlayer_1
nodes [{id=WavePlayer_1, properties {bLooping=true, soundWaveAssetPtr=&/Crucible_Drain_Steam_01}}]
properties {
 bOverrideAttenuation=true
 maxDistance=2900
 soundClassObject=&Fort_Ambient_Athena
 volumeMultiplier=0.2
 attenuationOverrides {dBAttenuationAtMax=-40, distanceAlgorithm=NaturalSound, falloffDistance=1800, stereoSpread=500, attenuationShapeExtents {x=200}}
}
