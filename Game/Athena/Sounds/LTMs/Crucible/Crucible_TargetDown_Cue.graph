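# Sound cue for a target being downed in the Crucible LTM, combining impact, score and hit-marker feedback.
# A Mixer layers a random pitched-up bullseye impact, a random respawn point-increase sting and a briefly delayed big-hit marker, each at its own level.
# A root Modulator slightly varies overall volume, and the cue routes to a 2D hit-notification sound class.
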SoundCue /Game/Athena/Sounds/LTMs/Crucible/Crucible_TargetDown_Cue
root=Modulator_0
nodes [
 {id=Delay_1, children [WavePlayer_5], properties {delayMax=0.5, delayMin=0.5}}
 {id=Mixer_1, children [Modulator_1,Random_2,Delay_1], properties {inputVolume [0.5,1,0.4]}}
 {id=Modulator_0, children [Mixer_1], properties {volumeMax=1.1, volumeMin=0.9}}
 {id=Modulator_1, children [Random_0], properties {pitchMax=1.3, pitchMin=1.3}}
 {id=Random_0, children [WavePlayer_1,WavePlayer_2,WavePlayer_3,WavePlayer_4], properties {weights [1,1,1,1]}}
 {id=Random_2, children [WavePlayer_8, WavePlayer_12], properties {weights [1,1]}}
 {id=WavePlayer_1, properties {soundWaveAssetPtr=/Game/Sounds/Creative/Toys/ShootingTargets/Target_Impact_Bullseye_01}}
 {id=WavePlayer_12, properties {soundWaveAssetPtr=/Game/Athena/Sounds/LTMs/Respawn/LTM_Respawn_Point_Increase_04}}
 {id=WavePlayer_2, properties {soundWaveAssetPtr=/Game/Sounds/Creative/Toys/ShootingTargets/Target_Impact_Bullseye_02}}
 {id=WavePlayer_3, properties {soundWaveAssetPtr=/Game/Sounds/Creative/Toys/ShootingTargets/Target_Impact_Bullseye_03}}
 {id=WavePlayer_4, properties {soundWaveAssetPtr=/Game/Sounds/Creative/Toys/ShootingTargets/Target_Impact_Bullseye_04}}
 {id=WavePlayer_5, properties {soundWaveAssetPtr=/Game/Athena/Sounds/HitNotifications/sfx_hitmarker_bighit}}
 {id=WavePlayer_8, properties {soundWaveAssetPtr=/Game/Athena/Sounds/LTMs/Respawn/LTM_Respawn_Point_Increase_05}}
]
properties {
 attenuationSettings=&/Crucible_TargetSpawned_Att
 maxDistance=3000
 soundClassObject=&Fort_2DHitNotifications_Instigate_PSM
 volumeMultiplier=0.6
 attenuationOverrides {bAttenuate=false, bSpatialize=false}
}
