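# Sound cue for eliminating a flyer target in the Crucible LTM.
# Mixes a pitched-up random bullseye impact, a random respawn point-increase sound and a briefly delayed flyer-specific wave, with the middle layer at full level and the others reduced.
# A root modulator slightly varies overall volume.
# It shares the Crucible target-spawn attenuation and routes through a 2D hit-notification sound class.
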SoundCue /Game/Athena/Sounds/LTMs/Crucible/Crucible_TargetEliminated_Flyer_Cue
root=Modulator_0
nodes [
 {id=Delay_2, children [WavePlayer_6], properties {delayMax=0.1, delayMin=0.1}}
 {id=Mixer_1, children [Modulator_2,Random_3,Delay_2], properties {inputVolume [0.5,1,0.5]}}
 {id=Modulator_0, children [Mixer_1], properties {volumeMax=1.1, volumeMin=0.9}}
 {id=Modulator_2, children [Random_1], properties {pitchMax=1.3, pitchMin=1.3}}
 {id=Random_1, children [WavePlayer_7, WavePlayer_9, WavePlayer_10, WavePlayer_11], properties {weights [1,1,1,1]}}
 {id=Random_3, children [WavePlayer_13, WavePlayer_14], properties {weights [1,1]}}
 {id=WavePlayer_10, properties {soundWaveAssetPtr=/Game/Sounds/Creative/Toys/ShootingTargets/Target_Impact_Bullseye_03}}
 {id=WavePlayer_11, properties {soundWaveAssetPtr=/Game/Sounds/Creative/Toys/ShootingTargets/Target_Impact_Bullseye_04}}
 {id=WavePlayer_13, properties {soundWaveAssetPtr=/Game/Athena/Sounds/LTMs/Respawn/LTM_Respawn_Point_Increase_05}}
 {id=WavePlayer_14, properties {soundWaveAssetPtr=/Game/Athena/Sounds/LTMs/Respawn/LTM_Respawn_Point_Increase_04}}
 {id=WavePlayer_6, properties {soundWaveAssetPtr=&/Crucible_TargetEliminated_Flyer_01}}
 {id=WavePlayer_7, properties {soundWaveAssetPtr=/Game/Sounds/Creative/Toys/ShootingTargets/Target_Impact_Bullseye_01}}
 {id=WavePlayer_9, properties {soundWaveAssetPtr=/Game/Sounds/Creative/Toys/ShootingTargets/Target_Impact_Bullseye_02}}
]
properties {
 attenuationSettings=&/Crucible_TargetSpawned_Att
 maxDistance=3000
 soundClassObject=&Fort_2DHitNotifications_Instigate_PSM
 volumeMultiplier=1
}
